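# EQS query for the Medic NPC specialist that finds the closest squad member in need of healing.
# It generates human-player squad members, excluding the querier, and filters out anyone tagged DBNO.
# A health and shield test keeps candidates by percentage thresholds, then a 3D distance test limits range and scores closer ones higher.
# The upper thresholds and distance bounds bind to float-curve-over-game-difficulty providers.
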EnvQuery /NPCSpecialists/NPCs/Medic/EQS/EQS_FindClosestSquadMemberToHeal
options [{
  generator {
   name=FortQueryGenerator_SquadMembers_0
   class=FortQueryGenerator_SquadMembers
   properties {
    bAutoSortTests=true
    bCanRunAsync=false
    envQueryResultNormalizationOption=Default
    includeSelf=false
    lookingForAiPawns=false
    lookingForHumanPlayers=true
    lookingForNpcs=false
    lookingForPlayerBots=false
    optionName=""
   }
  }
  tests [
   EnvQueryTest_GameplayTags_1 {
    class=EnvQueryTest_GameplayTags
    properties {
     bDefineReferenceValue=false
     bRejectIncompatibleItems=false
     bSearchComponentsForTags=false
     boolValue=true
     clampMaxType=None
     clampMinType=None
     filterType=Match
     floatValueMax=0
     floatValueMin=0
     multipleContextFilterOp=AllPass
     multipleContextScoreOp=AverageScore
     normalizationType=Absolute
     referenceValue=0
     scoreClampMax=0
     scoreClampMin=0
     scoringEquation=Constant
     scoringFactor=1
     testComment=""
     testPurpose=Filter
     tagQueryToMatch {
      autoDescription=" NONE( Gameplay.Status.DBNO )"
      tokenStreamVersion=0
      userDescription=""
      queryTokenStream [0,1,3,1,0]
      tagDictionary [{tagName=Gameplay.Status.DBNO}]
     }
    }
   }
   FortQueryTest_HealthAndShield_0 {
    class=FortQueryTest_HealthAndShield
    properties {
     bConsiderHealth=true
     bConsiderShield=true
     bDefineReferenceValue=false
     bHealthAsPercent=true
     bShieldAsPercent=true
     boolValue=true
     clampMaxType=None
     clampMinType=None
     filterType=Match
     floatValueMax=0
     floatValueMin=0
     healthMin=0
     multipleContextFilterOp=AllPass
     multipleContextScoreOp=AverageScore
     normalizationType=Absolute
     referenceValue=0
     scoreClampMax=0
     scoreClampMin=0
     scoringEquation=Constant
     scoringFactor=1
     shieldMin=0
     testComment=""
     testPurpose=Filter
     healthMax {
      dataBinding=&self:FortQueryTest_HealthAndShield_0.FortAIDataProvider_FloatCurveOverGameDifficulty_2
      dataField=FloatValue
      defaultValue=0.9
     }
     shieldMax {
      dataBinding=&self:FortQueryTest_HealthAndShield_0.FortAIDataProvider_FloatCurveOverGameDifficulty_1
      dataField=FloatValue
      defaultValue=0.9
     }
    }
   }
   EnvQueryTest_Distance_1 {
    class=EnvQueryTest_Distance
    properties {
     clampMaxType=SpecifiedValue
     distanceTo=&EnvQueryContext_Querier
     filterType=Range
     floatValueMin=0
     scoringEquation=InverseLinear
     scoringFactor=1
     testComment=Prefer closer items
     testMode=Distance3D
     testPurpose=FilterAndScore
     floatValueMax {
      dataBinding=&self:EnvQueryTest_Distance_1.FortAIDataProvider_FloatCurveOverGameDifficulty_3
      dataField=FloatValue
      defaultValue=2500
     }
     scoreClampMax {
      dataBinding=&self:EnvQueryTest_Distance_1.FortAIDataProvider_FloatCurveOverGameDifficulty_2
      dataField=FloatValue
      defaultValue=2500
     }
     scoreClampMin {
      dataBinding=&self:EnvQueryTest_Distance_1.FortAIDataProvider_FloatCurveOverGameDifficulty_1
      dataField=FloatValue
      defaultValue=0
     }
    }
   }
  ]
 }]
