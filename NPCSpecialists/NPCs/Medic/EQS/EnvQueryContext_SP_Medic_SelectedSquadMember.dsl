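; EQS context for the Medic NPC specialist that supplies the selected squad member as a single actor to environment queries.
; It casts the querier to a Fort player pawn, then validates its AI controller and blackboard.
; The actor comes from the blackboard object stored under the TargetActor key, which is held in a local variable, and is cast to Actor before being returned.
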(fn ProvideSingleActor
  (params (QuerierObject :type Object%20Reference) (QuerierActor :type Actor%20Object%20Reference))
  (bind _asfort_player_pawn (Utilities|Casting|CastToFortPlayerPawn QuerierActor)
    (:then
      (Utilities|IsValid (AI|GetAIController _asfort_player_pawn)
        (:Is%20Valid
          (Utilities|IsValid (AI|GetBlackboard (AI|GetAIController _asfort_player_pawn))
            (:Is%20Valid
              (bind _asactor (Utilities|Casting|CastToActor (AI|Components|Blackboard|GetValueasObject (AI|GetBlackboard (AI|GetAIController _asfort_player_pawn)) (LocalVariables|GetBBKey_TargetActor)))
                (:then
                  (return _asactor))
                (:CastFailed
                  _)))
            (:Is%20Not%20Valid
              _))
          (return 0))
        (:Is%20Not%20Valid
          _)))
    (:CastFailed
      _)))
